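# Additive unlit Niagara material for a glider headlight beam, part of the StallionSmoke damage-state VFX.
# Dynamic parameters skew and vertically offset vertices along instance axes, weighted by normal alignment to instance up.
# A refined, panning flame texture tinted by particle color drives emissive.
# Opacity multiplies particle alpha, a view-dependent thickness profile fading along UV length, and camera-facing falloff, with depth fade skipped on Low and Medium quality.
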Material /SpireSharedAssets/VFX/DamageState/M_Glider_StallionSmoke_Headlight_Beam
graph {
 properties {bUsedWithNiagaraSprites=true, blendMode=BLEND_Additive, pixelDepthOffsetMode=PDOM_Legacy, shadingModel=MSM_Unlit, usedShadingModels=Unlit}
 nodes [
  n1 = TextureCoordinate()
  n2 = ConstantBiasScale(None=ComponentMask(None=@n1) {r=true}) {bias=-0.5, scale=2}
  n3 = DynamicParameter() {paramNames [Skew, Vertical Offset, RDM, Param4]}
  n4 = Multiply(A=@n2, B=@n3.Skew)
  n5 = VertexNormalWS()
  n6 = Transform(None=Constant3Vector() {constant=<rgba 0 0 1 0>}) {transformSourceType=TRANSFORMSOURCE_Instance}
  n7 = Power(Base=Abs(None=DotProduct(A=@n5, B=@n6)), Exp=ScalarParameter() {defaultValue=2, parameterName=Base Thickness Lerp Exponent})
  n8 = Constant3Vector() {constant=<rgba 0 1 0 0>}
  Add(
   A=Multiply(A=Multiply(A=@n4, B=@n7), B=Transform(None=@n8) {transformSourceType=TRANSFORMSOURCE_Instance})
   B=Multiply(A=LinearInterpolate(A=@n3.Vertical Offset, Alpha=@n7) {constB=0}, B=@n6)
  )
  n9 = TextureObjectParameter() {
   group=EnergyFlow
   parameterName=EnergyFlow_Texture
   samplerType=SAMPLERTYPE_LinearGrayscale
   sortPriority=0
   texture=/Game/Effects/Fort_Effects/Emotes/UltraLight/Texture/T_FluidFlameTile_Greyscale
  }
  n10 = VectorParameter() {group=EnergyFlow, parameterName=EnergyFlow_UVControl, channelNames {a=VPan, b=UPan, g=VScale, r=UScale}, defaultValue=<rgba 0 0.35 0 -0.03>}
  n11 = TextureCoordinate()
  n12 = LinearInterpolate(
   A=ScalarParameter() {defaultValue=3.743, parameterName=EnergyFlowScaleA}
   B=@n10.UScale
   Alpha=Power(
    Base=MaterialFunctionCall(Float2=@n11) {materialFunction=&BreakOutFloat2Components}.G
    Exp=ScalarParameter() {defaultValue=0.4623, parameterName=TP}
   )
  )
  n13 = Time() {bOverride_Period=true, period=6000}
  n14 = DynamicParameter() {paramNames [Skew, Vertical Offset, RDM, Param4]}
  n15 = VectorParameter() {parameterName=EnergyFlow_Refine, defaultValue=<rgba 2.5943 20.8494 2 0.5>}
  n16 = MaterialFunctionCall(
   Texture=MaterialFunctionCall!(
    Texture=@n9
    AdditionalUVOffset
    UVControl_RGB=MaterialFunctionCall(X=@n12, Y=@n10.VScale, Z=@n10.VScale) {materialFunction=&MakeFloat3}.Result
    UVControl_A=@n10.VPan
    "[Override]TimeInput"=Add(A=@n13, B=@n14.RDM)
    "[Override]ScaleUVByCenter"=StaticBool() {value=true}
    "[Override]UVInput"
   ) {materialFunction=/Game/Effects/Niagara/MasterMaterials/Functions/Common/FXCommon_TextureSample_V2}.R
   RGB=@n15.RGB
   A=@n15.RGB
  ) {materialFunction=&MF_FX_Refine}
  n17 = ParticleColor()
  MaterialFunctionCall(Emissive=Multiply(A=Add(A=@n16.Result), B=@n17.RGB)) {materialFunction=&MF_FX_EmissiveAdjustments}
  n18 = ComponentMask(None=@n1) {g=true}
  n19 = OneMinus(None=@n18)
  n20 = ScalarParameter() {defaultValue=0.1, parameterName=Base Thickness from Side}
  n21 = ScalarParameter() {defaultValue=0.4, parameterName=Base Thickness from Above}
  n22 = OneMinus(None=Saturate(None=LinearInterpolate(A=@n20, B=@n21, Alpha=@n7)))
  n23 = Divide(A=@n2, B=OneMinus(None=Multiply(A=@n19, B=@n22)))
  n24 = OneMinus(None=Saturate(None=DotProduct(A=@n23, B=@n23)))
  n25 = Power(Base=Saturate(None=SquareRoot(None=@n24)), Exp=ScalarParameter() {defaultValue=5, parameterName=Edge Falloff})
  n26 = MaterialFunctionCall!(Gradient=@n18, Center Point=ScalarParameter() {defaultValue=0.05, parameterName=Base Fade}) {materialFunction=/Game/Effects/Fort_Effects/MaterialFunctions/MF_GradientReflected}
  n27 = DepthFade(Opacity=@n26.Result, FadeDistance=ScalarParameter() {defaultValue=11, parameterName=Depth Fade Distance})
  n28 = QualitySwitch(Default=@n27, Low=@n26.Result, High=@n27, Medium=@n26.Result, Epic=@n27)
  n29 = Multiply(A=@n28, B=@n28)
  n30 = Abs(None=DotProduct(A=VertexNormalWS(), B=CameraVectorWS()))
  Multiply(A=@n17.A, B=Multiply(A=Multiply(A=@n25, B=@n29), B=@n30))
  Power(Base=@n28, Exp=ScalarParameter() {defaultValue=2, parameterName=Length Falloff}) {desc=Replacing with multiply}
  Power(Base=@n30, Exp=ScalarParameter() {defaultValue=1, parameterName=Fresnel Exponent}) {desc=Skipping}
 ]
}
